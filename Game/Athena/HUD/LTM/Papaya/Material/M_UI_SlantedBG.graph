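# Translucent UI material for the Papaya LTM HUD, drawing a slanted background panel with a colored border.
# Two rotated UV gradients, smoothstepped and multiplied, form the slanted silhouette and an inset inner region.
# A vertical gradient mask on 9-slice UVs, thresholded by BorderMask, limits where the border shows when blending BorderColor with InnerColor.
# Opacity is the silhouette, reduced inside the fill in proportion to InnerColor alpha.
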Material /Game/Athena/HUD/LTM/Papaya/Material/M_UI_SlantedBG
graph {
 properties {blendMode=BLEND_Translucent, materialDomain=MD_UI, pixelDepthOffsetMode=PDOM_Legacy, usedShadingModels=Default Lit}
 nodes [
  n1 = VectorParameter() {parameterName=InnerColor, defaultValue=<rgba 0 0 0 0.5>}
  n2 = TextureSample(UVs=MaterialFunctionCall() {materialFunction=&GetUserInterfaceUV}.9-Slice UV) {samplerType=SAMPLERTYPE_Masks, texture=/Game/Global/Textures/Default/Masks_and_Gradients/T_VertGradient_MASK}
  n3 = ScalarParameter() {parameterName=EdgeSoftness}
  n4 = LinearInterpolate(A=Multiply(A=@n3) {constB=-1}, Alpha=ScalarParameter() {parameterName=BorderMask, sliderMax=1}) {constA=-0.2}
  n5 = Add(A=@n4, B=@n3)
  n6 = TextureCoordinate()
  n7 = VectorParameter() {parameterName=Center, defaultValue=<rgba 0.5 1 0 1>}
  n8 = ComponentMask(None=MaterialFunctionCall(
    UVs=@n6
    Rotation Center=@n7.RGB
    "Rotation Angle (0-1)"=ScalarParameter() {defaultValue=0.02, parameterName="Rotation Angle (0-1) (S)", sliderMax=1}
   ) {materialFunction=&CustomRotator}.Rotated Values) {r=true}
  n9 = ScalarParameter() {defaultValue=0.2, parameterName=Edge}
  n10 = Add(A=@n9, B=ScalarParameter() {defaultValue=0.2, parameterName=BorderWidth})
  n11 = Add(A=@n10, B=ScalarParameter() {parameterName=EdgeSoftness})
  n12 = ComponentMask(None=MaterialFunctionCall(UVs=@n6, Rotation Center=@n7.RGB, "Rotation Angle (0-1)"=ScalarParameter() {defaultValue=0.48, parameterName=Rotation Angle 2, sliderMax=1}) {materialFunction=&CustomRotator}.Rotated Values) {r=true}
  n13 = ScalarParameter() {defaultValue=0.2, parameterName=Edge}
  n14 = Add(A=@n13, B=ScalarParameter() {defaultValue=0.2, parameterName=BorderWidth})
  n15 = Add(A=@n14, B=ScalarParameter() {parameterName=EdgeSoftness})
  n16 = Max(
   A=OneMinus(None=MaterialFunctionCall(Alpha=@n2.R, Min=@n4, Max=@n5) {materialFunction=&SmoothStep}.Result)
   B=Multiply(
    A=MaterialFunctionCall(Alpha=@n8, Min=@n10, Max=@n11) {materialFunction=&SmoothStep}.Result
    B=MaterialFunctionCall(Alpha=@n12, Min=@n14, Max=@n15) {materialFunction=&SmoothStep}.Result
   )
  )
  LinearInterpolate(A=VectorParameter() {parameterName=BorderColor, defaultValue=<rgba 0 1 0.3328 1>}.RGB, B=@n1.RGB, Alpha=@n16)
  n17 = Add(A=@n9, B=ScalarParameter() {parameterName=EdgeSoftness})
  n18 = Add(A=@n13, B=ScalarParameter() {parameterName=EdgeSoftness})
  Multiply(
   A=Multiply(
    A=MaterialFunctionCall(Alpha=@n8, Min=@n9, Max=@n17) {materialFunction=&SmoothStep}.Result
    B=MaterialFunctionCall(Alpha=@n12, Min=@n13, Max=@n18) {materialFunction=&SmoothStep}.Result
   )
   B=OneMinus(None=Multiply(A=@n16, B=@n1.A) {constB=0.5})
  )
 ]
}
